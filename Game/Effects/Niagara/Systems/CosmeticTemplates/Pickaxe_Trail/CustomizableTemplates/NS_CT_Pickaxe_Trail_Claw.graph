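# Customizable Niagara template for claw pickaxe trails, drawn as ribbons from skeletal mesh bones or sockets through FN AnimTrail modules.
# Trail spawns one ribbon per finger via User.NumberOfFingers; Trail2 adds three thin ribbons and MiddleTrail a wider one by default.
# User parameters set each layer's color, width and lifetime, while UseTrail2 and UseTrail3 gate the extra layers by killing particles.
# Curves shape fade and width, with time-of-day tinting on all layers.
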NiagaraSystem /Game/Effects/Niagara/Systems/CosmeticTemplates/Pickaxe_Trail/CustomizableTemplates/NS_CT_Pickaxe_Trail_Claw
dependencies {
 usedDataInterfaces [&NiagaraDataInterfaceColorCurve, &NiagaraDataInterfaceCurve, &NiagaraDataInterfaceSkeletalMesh]
 usedDynamicInputs []
 usedModules [
  &self:SystemSpawnScript.SystemScriptSource.SystemScriptGraph.NiagaraNodeAssignment_0.SetVariables_#1
  &SystemState
 ]
}
emitters {
 MiddleTrail {
  emitterUpdateScript [
   &EmitterState(Life Cycle Mode=::NewEnumerator1, Loop Behavior=::NewEnumerator1, Loop Duration=0.28, UseLoopDelay=true)
   &FN_SpawnAnimTrail(Switch on ENiagara_AnimTrail_Source:ENiagara_AnimTrail_Source=::NewEnumerator1)
  ]
  particleSpawnScript [
   &FN_InitializeAnimTrailBoneIndecies(Second Bone or Socket Index=0)
   &FN_InitializeAnimTrail(AnimTrailWidth=$User.TrailWidth3, Color=$User.TrailColor3, Lifetime=$User.TrailLifetime3, SkeletalMesh=$System.Skeletal Mesh Reader)
  ]
  particleUpdateScript [
   &FN_UpdateAnimTrail(SkeletalMesh=$System.Skeletal Mesh Reader)
   &ParticleState()
   &ScaleColor(
    Scale Alpha=&Multiply_Float(
     A=&Multiply_Float(
      A=&FloatFromCurve(
       FloatCurve=DataInterface() {propertyValues {Curve {keys [
           {arriveTangent=-0.0116, interpMode=RCIM_Cubic, leaveTangent=-2.9135, tangentMode=RCTM_Break, time=0.3, value=1}
           {arriveTangent=-0.029, interpMode=RCIM_Cubic, tangentMode=RCTM_Break, time=1, value=0}
          ]}}}
       Scale Curve=0.8
      )
      B=&FloatFromCurve(FloatCurve=DataInterface() {propertyValues {Curve {keys [<at 0.05 0>,<at 0.1 1>]}}}, CurveIndex=$Emitter.Age)
     )
     B=&FN_FrontEndFloatModifier(Gameplay=1, Front End=0.5)
    )
    Scale RGB=&MakeVectorFromLinearColorRGB(Color=&FNTimeOfDayColor(Night Color=<rgba 0.3508 0.5387 0.7448 1>, Morning Color=<rgba 1 0.6659 0.6436 1>, Day Color=<rgba 1 1 1 1>, Evening Color=<rgba 1 1 1 1>))
    ScaleA=true
    ScaleRGB=true
   )
   &ScaleRibbonWidth(Ribbon Width Scale=&FloatFromCurve(FloatCurve=DataInterface() {propertyValues {Curve {
        defaultValue=1.2
        keys [
         {arriveTangent=4.1785, interpMode=RCIM_Cubic, leaveTangent=4.1785, tangentMode=RCTM_User, time=0, value=0}
         {arriveTangent=1.1179, interpMode=RCIM_Cubic, leaveTangent=1.1179, tangentMode=RCTM_User, time=0.2042, value=0.6228}
         {arriveTangent=0.0224, interpMode=RCIM_Cubic, leaveTangent=0.0224, tangentMode=RCTM_User, time=1, value=1}
        ]
       }}}))
   &KillParticles(Kill Particles=&SetBoolByFloatComparison(A=$User.UseTrail3, B=1, Comparison Type=::NewEnumerator6))
  ]
  rendererClasses [&NiagaraRibbonRendererProperties]
  renderers [{rendererClass=&NiagaraRibbonRendererProperties}]
 }
 Trail {
  emitterUpdateScript [
   &EmitterState(Life Cycle Mode=::NewEnumerator1, Loop Behavior=::NewEnumerator1, Loop Delay=0.05, Loop Duration=0.25, UseLoopDelay=true)
   &FN_SpawnAnimTrail(
    Number of Anim Trails:NiagaraInt32=&MakeIntFromFloat(Float=$User.NumberOfFingers)
    Switch on ENiagara_AnimTrail_Source:ENiagara_AnimTrail_Source=::NewEnumerator1
   )
  ]
  particleSpawnScript [
   &FN_InitializeAnimTrailBoneIndecies(Second Bone or Socket Index=0)
   &FN_InitializeAnimTrailBoneIndecies(AnimTrail Selection=1, First Bone or Socket Index=1) {module=FN_InitializeAnimTrailBoneIndecies001}
   &FN_InitializeAnimTrailBoneIndecies(AnimTrail Selection=2, First Bone or Socket Index=2, Second Bone or Socket Index=2) {module=FN_InitializeAnimTrailBoneIndecies002}
   &FN_InitializeAnimTrailBoneIndecies(AnimTrail Selection=3, First Bone or Socket Index=3, Second Bone or Socket Index=3) {module=FN_InitializeAnimTrailBoneIndecies003}
   &FN_InitializeAnimTrail(AnimTrailWidth=$User.TrailWidth1, Color=$User.TrailColor1, Lifetime=$User.TrailLifetime1, SkeletalMesh=$System.Skeletal Mesh Reader)
  ]
  particleUpdateScript [
   &ParticleState()
   &FN_UpdateAnimTrail(SkeletalMesh=$System.Skeletal Mesh Reader)
   &ScaleColor(
    Scale Alpha=&FloatFromCurve(
     FloatCurve=DataInterface() {propertyValues {Curve {keys [
         {arriveTangent=-3.4238, interpMode=RCIM_Cubic, leaveTangent=-3.4238, tangentMode=RCTM_User, time=0, value=1}
         {interpMode=RCIM_Cubic, tangentMode=RCTM_User, time=1, value=0}
        ]}}}
     Scale Curve=2
    )
    Scale RGB=&MakeVectorFromLinearColorRGB(Color=&ColorFromCurve(ColorCurve=DataInterface() {propertyValues {
        AlphaCurve {keys []}
        BlueCurve {keys [<at 0.008 2>, {arriveTangent=-4.3928, leaveTangent=-4.3928, time=0.5321, value=0.8038}, <at 0.6889 0.0434>]}
        GreenCurve {keys [<at 0.008 2>, {arriveTangent=-4.3928, leaveTangent=-4.3928, time=0.5321, value=0.8038}, <at 0.6889 0.0434>]}
        RedCurve {keys [<at 0.008 2>, {arriveTangent=-4.3928, leaveTangent=-4.3928, time=0.5321, value=0.8038}, <at 0.6889 0.0434>]}
       }}))
    ScaleA=true
    ScaleRGB=true
   )
   &ScaleRibbonWidth(Ribbon Width Scale=&FloatFromCurve(FloatCurve=DataInterface() {propertyValues {Curve {keys [{arriveTangent=-0.0042, interpMode=RCIM_Cubic, leaveTangent=-0.0042, tangentMode=RCTM_User, time=0, value=1}, <at 1 0>]}}}))
   &ScaleColor(
    Scale RGB=&MakeVectorFromLinearColorRGB(Color=&FNTimeOfDayColor(Night Color=<rgba 0.3508 0.5387 0.7448 1>, Morning Color=<rgba 1 0.6659 0.6436 1>, Day Color=<rgba 1 1 1 1>, Evening Color=<rgba 1 1 1 1>))
    ScaleA=true
    ScaleRGB=true
   ) {module=ScaleColor001}
  ]
  rendererClasses [&NiagaraRibbonRendererProperties]
  renderers [{rendererClass=&NiagaraRibbonRendererProperties}]
 }
 Trail2 {
  emitterUpdateScript [
   &EmitterState(Life Cycle Mode=::NewEnumerator1, Loop Behavior=::NewEnumerator1, Loop Delay=0.05, Loop Duration=0.25, UseLoopDelay=true)
   &FN_SpawnAnimTrail(Number of Anim Trails=3, Switch on ENiagara_AnimTrail_Source:ENiagara_AnimTrail_Source=::NewEnumerator1)
  ]
  particleSpawnScript [
   &FN_InitializeAnimTrailBoneIndecies(Second Bone or Socket Index=0)
   &FN_InitializeAnimTrailBoneIndecies(AnimTrail Selection=1, First Bone or Socket Index=1) {module=FN_InitializeAnimTrailBoneIndecies001}
   &FN_InitializeAnimTrailBoneIndecies(AnimTrail Selection=2, First Bone or Socket Index=2, Second Bone or Socket Index=2) {module=FN_InitializeAnimTrailBoneIndecies002}
   &FN_InitializeAnimTrailBoneIndecies(AnimTrail Selection=3, First Bone or Socket Index=3, Second Bone or Socket Index=3) {module=FN_InitializeAnimTrailBoneIndecies003}
   &FN_InitializeAnimTrail(AnimTrailWidth=$User.TrailWidth2, Color=$User.TrailColor2, Lifetime=$User.TrailLifetime2, SkeletalMesh=$System.Skeletal Mesh Reader)
  ]
  particleUpdateScript [
   &ParticleState()
   &FN_UpdateAnimTrail(SkeletalMesh=$System.Skeletal Mesh Reader)
   &ScaleColor(
    Scale Alpha=&FloatFromCurve(
     FloatCurve=DataInterface() {propertyValues {Curve {keys [
         {arriveTangent=-3.3229, interpMode=RCIM_Cubic, leaveTangent=-3.3229, tangentMode=RCTM_User, time=0, value=1}
         {interpMode=RCIM_Cubic, tangentMode=RCTM_User, time=1, value=0}
        ]}}}
     Scale Curve=2
    )
    Scale RGB=&MakeVectorFromLinearColorRGB(Color=&ColorFromCurve(ColorCurve=DataInterface() {propertyValues {
        AlphaCurve {keys []}
        BlueCurve {keys [<at 0.008 2>, {arriveTangent=-4.3928, leaveTangent=-4.3928, time=0.5321, value=0.8038}, <at 0.6889 0.0434>]}
        GreenCurve {keys [<at 0.008 2>, {arriveTangent=-4.3928, leaveTangent=-4.3928, time=0.5321, value=0.8038}, <at 0.6889 0.0434>]}
        RedCurve {keys [<at 0.008 2>, {arriveTangent=-4.3928, leaveTangent=-4.3928, time=0.5321, value=0.8038}, <at 0.6889 0.0434>]}
       }}))
    ScaleA=true
    ScaleRGB=true
   )
   &ScaleRibbonWidth(Ribbon Width Scale=&FloatFromCurve(FloatCurve=DataInterface() {propertyValues {Curve {keys [{arriveTangent=-0.0042, interpMode=RCIM_Cubic, leaveTangent=-0.0042, tangentMode=RCTM_User, time=0, value=1}, <at 1 0>]}}}))
   &KillParticles(Kill Particles=&SetBoolByFloatComparison(A=$User.UseTrail2, B=1, Comparison Type=::NewEnumerator6))
   &ScaleColor(
    Scale RGB=&MakeVectorFromLinearColorRGB(Color=&FNTimeOfDayColor(Night Color=<rgba 0.3508 0.5387 0.7448 1>, Morning Color=<rgba 1 0.6659 0.6436 1>, Day Color=<rgba 1 1 1 1>, Evening Color=<rgba 1 1 1 1>))
    ScaleA=true
    ScaleRGB=true
   ) {module=ScaleColor001}
  ]
  rendererClasses [&NiagaraRibbonRendererProperties]
  renderers [{rendererClass=&NiagaraRibbonRendererProperties}]
 }
}
summary {
 systemName=NS_CT_Pickaxe_Trail_Claw
 emitters [
  {emitterName=Trail, rendererClasses [&NiagaraRibbonRendererProperties]}
  {emitterName=MiddleTrail, rendererClasses [&NiagaraRibbonRendererProperties]}
  {emitterName=Trail2, rendererClasses [&NiagaraRibbonRendererProperties]}
 ]
 userVariables [
  User.bIsFrontEnd {type=NiagaraBool}
  User.bIsFrontEndPreview {type=NiagaraBool}
  User.MITrail1 {
   type=MaterialInterface
   description=""
   defaultValue {
    struct=NiagaraExt_VariableValue_Object
    value {object=/Game/Effects/Niagara/MasterMaterials/Ribbon/MI_SmokeTrail_Long, objectClass=/Script/Engine.MaterialInterface}
   }
  }
  User.MITrail2 {
   type=MaterialInterface
   description=""
   defaultValue {
    struct=NiagaraExt_VariableValue_Object
    value {object=&MI_SmokeTrail_Long, objectClass=&MaterialInterface}
   }
  }
  User.NumberOfFingers {type=NiagaraFloat, defaultValue=3, description=""}
  User.RibbonIDOffset {type=NiagaraInt32}
  User.TrailColor1 {type=LinearColor, description="", defaultValue {struct=LinearColor, value=<rgba 1 1 1 0.5>}}
  User.TrailColor2 {type=LinearColor, description="", defaultValue {struct=LinearColor, value=<rgba 1 1 1 0.5>}}
  User.TrailColor3 {type=LinearColor, description="", defaultValue {struct=LinearColor, value=<rgba 1 1 1 0.5>}}
  User.TrailLifetime1 {type=NiagaraFloat, defaultValue=0.25, description=""}
  User.TrailLifetime2 {type=NiagaraFloat, defaultValue=0.25, description=""}
  User.TrailLifetime3 {type=NiagaraFloat, defaultValue=0.7, description=""}
  User.TrailWidth1 {type=NiagaraFloat, defaultValue=6, description=""}
  User.TrailWidth2 {type=NiagaraFloat, defaultValue=6, description=""}
  User.TrailWidth3 {type=NiagaraFloat, defaultValue=55, description=""}
  User.UseTrail2 {type=NiagaraFloat, defaultValue=0, description=""}
  User.UseTrail3 {type=NiagaraFloat, defaultValue=0, description=""}
 ]
}
