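# Customizable Niagara template for cosmetic offhand pickaxe trails, drawing up to three ribbon anim trails between socket pairs.
# The only enabled emitter, NS_CT_Trail_Simple, merges the three trails, using spawn group to select width, color and lifetime from User parameters and set visibility tags for six ribbon renderers.
# Particle color and alpha fade are scaled by time of day.
# The three older-style per-trail emitters remain present but disabled.
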NiagaraSystem /Game/Effects/Niagara/Systems/CosmeticTemplates/Pickaxe_Trail/CustomizableTemplates/NS_CT_Pickaxe_Trail_Offhand
dependencies {
 usedDataInterfaces [&NiagaraDataInterfaceCurve, &NiagaraDataInterfaceSocketReader]
 usedDynamicInputs [
  &Boolean_OrOperation
  &Boolean_AndOperation
  &Boolean_NotOperation
 ]
 usedModules [
  &self:SystemSpawnScript.SystemScriptSource.SystemScriptGraph.NiagaraNodeAssignment_1.SetVariables_#1
  &self:SystemSpawnScript.SystemScriptSource.SystemScriptGraph.NiagaraNodeAssignment_0.SetVariables_#2
  &self:SystemSpawnScript.SystemScriptSource.SystemScriptGraph.NiagaraNodeAssignment_2.SetVariables_#3
  &self:SystemSpawnScript.SystemScriptSource.SystemScriptGraph.NiagaraNodeAssignment_3.SetVariables_#4
  &self:SystemSpawnScript.SystemScriptSource.SystemScriptGraph.NiagaraNodeAssignment_4.SetVariables_#5
  &SystemState
 ]
}
emitters {
 NE_Trail_01 {
  bEnabled=false
  emitterSpawnScript [&FN_RibbonIDOffsetUtility()]
  emitterUpdateScript [
   &EmitterState(Inactive Response=::NewEnumerator1, Life Cycle Mode=::NewEnumerator1, Loop Behavior=::NewEnumerator1, Loop Duration=1, UseLoopDelay=false)
   &FN_SpawnAnimTrail()
  ]
  particleSpawnScript [
   &FN_InitializeAnimTrailBoneIndecies(
    First Bone or Socket Index:NiagaraInt32=&MakeIntFromFloat(Float=$User.FirstSocket1)
    Second Bone or Socket Index:NiagaraInt32=&MakeIntFromFloat(Float=$User.SecondSocket1)
   )
   &FN_InitializeAnimTrail(AnimTrailWidth=$User.TrailWidth1, Color=$User.TrailColor1, Lifetime=$User.TrailLifetime1)
  ]
  particleUpdateScript [
   &SolveForcesAndVelocity(Clamp Velocity=false, Limit Acceleration=false)
   &FN_UpdateAnimTrail() {module=FN_UpdateAnimTrail001}
   &ScaleColor(
    Scale Alpha=&FloatFromCurve(
     FloatCurve=DataInterface() {propertyValues {Curve {keys [{arriveTangent=-56.0888, interpMode=RCIM_Cubic, leaveTangent=-56.0888, tangentMode=RCTM_User, time=0.3, value=1}, <at 0.35 0 RCIM_Cubic>]}}}
     Scale Curve=&FNTimeOfDayScalar(Night Scale=$User.NightTOD, Morning Scale=1, Day Scale=$User.DayTOD, Evening Scale=1)
    )
    Scale RGB=&MakeVectorFromLinearColorRGB(Color=&FNTimeOfDayColor(Night Color=<rgba 0.3508 0.5387 0.7448 1>, Morning Color=<rgba 1 0.6659 0.6436 1>, Day Color=<rgba 1 1 1 1>, Evening Color=<rgba 1 1 1 1>))
    ScaleA=true
    ScaleRGB=true
   )
  ]
  rendererClasses [&NiagaraRibbonRendererProperties]
  renderers [{rendererClass=&NiagaraRibbonRendererProperties}]
 }
 NE_Trail_02 {
  bEnabled=false
  emitterSpawnScript [&FN_RibbonIDOffsetUtility()]
  emitterUpdateScript [
   &EmitterState(
    Inactive Response=::NewEnumerator1
    Life Cycle Mode=::NewEnumerator1
    Loop Behavior=::NewEnumerator1
    Loop Duration=$User.UseTrail2
    UseLoopDelay=false
   )
   &FN_SpawnAnimTrail()
  ]
  particleSpawnScript [
   &FN_InitializeAnimTrailBoneIndecies(
    First Bone or Socket Index:NiagaraInt32=&MakeIntFromFloat(Float=$User.FirstSocket2)
    Second Bone or Socket Index:NiagaraInt32=&MakeIntFromFloat(Float=$User.SecondSocket2)
   )
   &FN_InitializeAnimTrail(AnimTrailWidth=$User.TrailWidth2, Color=$User.TrailColor2, Lifetime=$User.TrailLifetime2)
  ]
  particleUpdateScript [
   &SolveForcesAndVelocity(Clamp Velocity=false, Limit Acceleration=false)
   &FN_UpdateAnimTrail() {module=FN_UpdateAnimTrail001}
   &ScaleColor(
    Scale Alpha=&FloatFromCurve(
     FloatCurve=DataInterface() {propertyValues {Curve {keys [{arriveTangent=-56.0888, interpMode=RCIM_Cubic, leaveTangent=-56.0888, tangentMode=RCTM_User, time=0.3, value=1}, <at 0.35 0 RCIM_Cubic>]}}}
     Scale Curve=&FNTimeOfDayScalar(Night Scale=$User.NightTOD, Morning Scale=1, Day Scale=$User.DayTOD, Evening Scale=1)
    )
    Scale RGB=&MakeVectorFromLinearColorRGB(Color=&FNTimeOfDayColor(Night Color=<rgba 0.3508 0.5387 0.7448 1>, Morning Color=<rgba 1 0.6659 0.6436 1>, Day Color=<rgba 1 1 1 1>, Evening Color=<rgba 1 1 1 1>))
    ScaleA=true
    ScaleRGB=true
   )
   &KillParticles(Kill Particles=&SetBoolByFloatComparison(A=$User.UseTrail2, B=0, Comparison Type=::NewEnumerator4))
  ]
  rendererClasses [&NiagaraRibbonRendererProperties]
  renderers [{rendererClass=&NiagaraRibbonRendererProperties}]
 }
 NE_Trail_03 {
  bEnabled=false
  emitterSpawnScript [&FN_RibbonIDOffsetUtility()]
  emitterUpdateScript [
   &EmitterState(
    Inactive Response=::NewEnumerator1
    Life Cycle Mode=::NewEnumerator1
    Loop Behavior=::NewEnumerator1
    Loop Duration=$User.UseTrail3
    UseLoopDelay=false
   )
   &FN_SpawnAnimTrail()
  ]
  particleSpawnScript [
   &FN_InitializeAnimTrailBoneIndecies(
    First Bone or Socket Index:NiagaraInt32=&MakeIntFromFloat(Float=$User.FirstSocket3)
    Second Bone or Socket Index:NiagaraInt32=&MakeIntFromFloat(Float=$User.SecondSocket3)
   )
   &FN_InitializeAnimTrail(AnimTrailWidth=$User.TrailWidth3, Color=$User.TrailColor3, Lifetime=$User.TrailLifetime3)
  ]
  particleUpdateScript [
   &SolveForcesAndVelocity(Clamp Velocity=false, Limit Acceleration=false)
   &FN_UpdateAnimTrail() {module=FN_UpdateAnimTrail001}
   &ScaleColor(
    Scale Alpha=&FloatFromCurve(
     FloatCurve=DataInterface() {propertyValues {Curve {keys [{arriveTangent=-56.0888, interpMode=RCIM_Cubic, leaveTangent=-56.0888, tangentMode=RCTM_User, time=0.3, value=1}, <at 0.35 0 RCIM_Cubic>]}}}
     Scale Curve=&FNTimeOfDayScalar(Night Scale=$User.NightTOD, Morning Scale=1, Day Scale=$User.DayTOD, Evening Scale=1)
    )
    Scale RGB=&MakeVectorFromLinearColorRGB(Color=&FNTimeOfDayColor(Night Color=<rgba 0.3508 0.5387 0.7448 1>, Morning Color=<rgba 1 0.6659 0.6436 1>, Day Color=<rgba 1 1 1 1>, Evening Color=<rgba 1 1 1 1>))
    ScaleA=true
    ScaleRGB=true
   )
   &KillParticles(Kill Particles=&SetBoolByFloatComparison(A=$User.UseTrail3, B=0, Comparison Type=::NewEnumerator4))
  ]
  rendererClasses [&NiagaraRibbonRendererProperties]
  renderers [{rendererClass=&NiagaraRibbonRendererProperties}]
 }
 NS_CT_Trail_Simple {
  emitterSpawnScript [&FN_RibbonIDOffsetUtility()]
  emitterUpdateScript [
   &EmitterState(Inactive Response=::NewEnumerator1, Life Cycle Mode=::NewEnumerator1, Loop Behavior=::NewEnumerator1, Loop Duration=1, UseLoopDelay=false)
   &FN_SpawnAnimTrail(Number of Anim Trails=3)
  ]
  particleSpawnScript [
   &FN_InitializeAnimTrailBoneIndecies(
    First Bone or Socket Index:NiagaraInt32=&MakeIntFromFloat(Float=$User.FirstSocket1)
    Second Bone or Socket Index:NiagaraInt32=&MakeIntFromFloat(Float=$User.SecondSocket1)
   )
   &FN_InitializeAnimTrailBoneIndecies(
    AnimTrail Selection=1
    First Bone or Socket Index:NiagaraInt32=&MakeIntFromFloat(Float=$User.FirstSocket2)
    Second Bone or Socket Index:NiagaraInt32=&MakeIntFromFloat(Float=$User.SecondSocket2)
   ) {module=FN_InitializeAnimTrailBoneIndecies001}
   &FN_InitializeAnimTrailBoneIndecies(
    AnimTrail Selection=2
    First Bone or Socket Index:NiagaraInt32=&MakeIntFromFloat(Float=$User.FirstSocket3)
    Second Bone or Socket Index:NiagaraInt32=&MakeIntFromFloat(Float=$User.SecondSocket3)
   ) {module=FN_InitializeAnimTrailBoneIndecies002}
   /SetVariables(Particles.VisibilityTag=&SpawnGroup()) {isSetParameters=true}
   &FN_InitializeAnimTrail(
    AnimTrailWidth=&MergeHelper_FloatSwitch(0=$User.TrailWidth1, 1=$User.TrailWidth2, 2=$User.TrailWidth3, 3=?, Index=&SpawnGroup(), Num=::NewEnumerator1)
    Color=&MergeHelper_ColorSwitch(0=$User.TrailColor1, 1=$User.TrailColor2, 2=$User.TrailColor3, 3=?, Index=&SpawnGroup(), Num=::NewEnumerator1)
    Lifetime=&MergeHelper_FloatSwitch(
     0=$User.TrailLifetime1
     1=$User.TrailLifetime2
     2=$User.TrailLifetime3
     3=?
     Index=&SpawnGroup()
     Num=::NewEnumerator1
    )
   )
   &AddVelocity(Velocity=<0 0 250>) {enabled=false}
  ]
  particleUpdateScript [
   &SolveForcesAndVelocity(Clamp Velocity=false, Limit Acceleration=false)
   &FN_UpdateAnimTrail() {module=FN_UpdateAnimTrail001}
   &ScaleColor(
    Scale Alpha=&FloatFromCurve(
     FloatCurve=DataInterface() {propertyValues {Curve {keys [{arriveTangent=-56.0888, interpMode=RCIM_Cubic, leaveTangent=-56.0888, tangentMode=RCTM_User, time=0.3, value=1}, <at 0.35 0 RCIM_Cubic>]}}}
     Scale Curve=&FNTimeOfDayScalar(Night Scale=$User.NightTOD, Morning Scale=1, Day Scale=$User.DayTOD, Evening Scale=1)
    )
    Scale RGB=&MakeVectorFromLinearColorRGB(Color=&FNTimeOfDayColor(Night Color=<rgba 0.3508 0.5387 0.7448 1>, Morning Color=<rgba 1 0.6659 0.6436 1>, Day Color=<rgba 1 1 1 1>, Evening Color=<rgba 1 1 1 1>))
    ScaleA=true
    ScaleRGB=true
   )
  ]
  rendererClasses [&NiagaraRibbonRendererProperties]
  renderers [
   {rendererClass=&NiagaraRibbonRendererProperties}
   {rendererClass=&NiagaraRibbonRendererProperties, rendererIndex=1}
   {rendererClass=&NiagaraRibbonRendererProperties, rendererIndex=2}
   {rendererClass=&NiagaraRibbonRendererProperties, rendererIndex=3}
   {rendererClass=&NiagaraRibbonRendererProperties, rendererIndex=4}
   {rendererClass=&NiagaraRibbonRendererProperties, rendererIndex=5}
  ]
 }
}
summary {
 systemName=NS_CT_Pickaxe_Trail_Offhand
 emitters [
  {bEnabled=false, emitterName=NE_Trail_01, rendererClasses [&NiagaraRibbonRendererProperties]}
  {bEnabled=false, emitterName=NE_Trail_02, rendererClasses [&NiagaraRibbonRendererProperties]}
  {bEnabled=false, emitterName=NE_Trail_03, rendererClasses [&NiagaraRibbonRendererProperties]}
  {emitterName=NS_CT_Trail_Simple, rendererClasses [&NiagaraRibbonRendererProperties]}
 ]
 userVariables [
  User.bIsFrontEnd {type=NiagaraBool}
  User.bIsFrontEndPreview {type=NiagaraBool}
  User.DayTOD {type=NiagaraFloat, defaultValue=1, description=""}
  User.FirstSocket1 {type=NiagaraFloat, defaultValue=0, description=""}
  User.FirstSocket2 {type=NiagaraFloat, defaultValue=2, description=""}
  User.FirstSocket3 {type=NiagaraFloat, defaultValue=4, description=""}
  User.MITrail1 {
   type=MaterialInterface
   description=""
   defaultValue {struct=NiagaraExt_VariableValue_Object, value {object=null, objectClass=/Script/Engine.MaterialInterface}}
  }
  User.MITrail2 {
   type=MaterialInterface
   description=""
   defaultValue {struct=NiagaraExt_VariableValue_Object, value {object=null, objectClass=&MaterialInterface}}
  }
  User.MITrail3 {
   type=MaterialInterface
   description=""
   defaultValue {struct=NiagaraExt_VariableValue_Object, value {object=null, objectClass=&MaterialInterface}}
  }
  User.NightTOD {type=NiagaraFloat, defaultValue=0.7, description=""}
  User.RibbonIDOffset {type=NiagaraInt32}
  User.SecondSocket1 {type=NiagaraFloat, defaultValue=1, description=""}
  User.SecondSocket2 {type=NiagaraFloat, defaultValue=3, description=""}
  User.SecondSocket3 {type=NiagaraFloat, defaultValue=5, description=""}
  User.ToggleScreen1 {type=NiagaraFloat, defaultValue=0, description=""}
  User.ToggleScreen2 {type=NiagaraFloat, defaultValue=0, description=""}
  User.ToggleScreen3 {type=NiagaraFloat, defaultValue=0, description=""}
  User.ToggleScreenAll {type=NiagaraFloat, defaultValue=0, description=""}
  User.TrailColor1 {type=LinearColor, description="", defaultValue {struct=LinearColor, value=<rgba 1 1 1 1>}}
  User.TrailColor2 {type=LinearColor, description="", defaultValue {struct=LinearColor, value=<rgba 1 1 1 1>}}
  User.TrailColor3 {type=LinearColor, description="", defaultValue {struct=LinearColor, value=<rgba 1 1 1 1>}}
  User.TrailLifetime1 {type=NiagaraFloat, defaultValue=0.22, description=""}
  User.TrailLifetime2 {type=NiagaraFloat, defaultValue=0.22, description=""}
  User.TrailLifetime3 {type=NiagaraFloat, defaultValue=0.22, description=""}
  User.TrailWidth1 {type=NiagaraFloat, defaultValue=1, description=""}
  User.TrailWidth2 {type=NiagaraFloat, defaultValue=1, description=""}
  User.TrailWidth3 {type=NiagaraFloat, defaultValue=1, description=""}
  User.UseTrail2 {type=NiagaraFloat, defaultValue=0, description=""}
  User.UseTrail3 {type=NiagaraFloat, defaultValue=0, description=""}
 ]
}
